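; Soccer minigame actor for Battle Pass S22 (Redoasis) with two goal volumes, a score sign and respawning balls.
; Server-side goal overlaps consume the pitch or toy ball, credit the scorer with quest stats and a recently-scored effect, then check the win limit.
; Score, location and winner RepNotifies drive sign materials and goal/win FX, with a winner triggering a delayed match restart.
; Balls leaving the field are destroyed and respawned.
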(event Custom|InitializeGoalCollision_Multicast
  (CallFunction|InitializeGoalCollision (Variables|Default|GetTeam_A_GoalOverlap) (Variables|Default|GetHideGoalA))
  (CallFunction|InitializeGoalCollision (Variables|Default|GetTeam_B_GoalOverlap) (Variables|Default|GetHideGoalB))
  (CallFunction|InitializeSignCollision))

(event EventBeginPlay
  ;[1 1 1] Check if In Creative Mode or not
  (bind _asfort_game_state_athena (Utilities|Casting|CastToFortGameStateAthena (Game|GetGameState)))
  (if (Variables|/Script/FortniteGame.FortGameStateAthena|GetbPlaylistDataIsLoaded _asfort_game_state_athena)
    (if (GameplayTags|HasTag (Class|FortGameStateAthena|GetAthenaPlaylistContextTags _asfort_game_state_athena) "(TagName=\"Athena.Creative\")")
      (Variables|QuestStuff|SetCreativeMode true)
      (else
        (Variables|QuestStuff|SetCreativeMode false)))
    (Networking|SwitchHasAuthority
      (:Authority
        (CallFunction|InitializeGoalCollisionMulticast) ; Collision responses aren't updated properly on construction, so we do it on BeginPlay with a multicast node to ensure clients receive the update
        (if (not (Variables|Default|GetHideSign))
          ;[1 1 1] Initialize Scoreboard Vars
          (Utilities|FlowControl|Delay 0.2)
          (Variables|SoccerGame|SetScore_Team_B 0)
          (Variables|SoccerGame|SetScore_Team_A 0))))
    (else
      (Playlist|BindEventtoOnCurrentPlaylistLoaded _asfort_game_state_athena (delegate Custom|PlaylistLoaded)))))

(event Custom|PlaylistLoaded
  (params (PlaylistName :type Name) (PlaylistContextTags :type Gameplay%20Tag%20Container%20Structure%20%28by%20ref%29))
  ;[1 1 1] Check if In Creative Mode or not
  (if (GameplayTags|HasTag PlaylistContextTags "(TagName=\"Athena.Creative\")")
    (Variables|QuestStuff|SetCreativeMode true)
    (else
      (Variables|QuestStuff|SetCreativeMode false)))
  (Networking|SwitchHasAuthority
    (:Authority
      (CallFunction|InitializeGoalCollisionMulticast) ; Collision responses aren't updated properly on construction, so we do it on BeginPlay with a multicast node to ensure clients receive the update
      (if (not (Variables|Default|GetHideSign))
        ;[1 1 1] Initialize Scoreboard Vars
        (Utilities|FlowControl|Delay 0.2)
        (Variables|SoccerGame|SetScore_Team_B 0)
        (Variables|SoccerGame|SetScore_Team_A 0)))))

(event OnComponentBeginOverlap(Team_A_GoalOverlap)
  (params (OverlappedComponent :type Primitive%20Component%20Object%20Reference) (OtherActor :type Actor%20Object%20Reference) (OtherComp :type Primitive%20Component%20Object%20Reference) (OtherBodyIndex :type Integer) (bFromSweep :type Boolean) (SweepResult :type Hit%20Result%20Structure%20%28by%20ref%29))
  (bind _playerwhoscored (Variables|QuestStuff|GetPlayerWhoScored))
  (bind _returnvalue (Fortnite|GetFortPlayerControllerfromActor _playerwhoscored))
  (bind _returnvalue_2 (Quests|GetQuestManager _returnvalue "Invalid"))
  ;[1 1 1] Player Scored on Goal A
  (Networking|SwitchHasAuthority
    (:Authority
      (bind _bwasballoverlap (CallFunction|CheckandHandleBallOverlap OtherActor))
      (Variables|Default|SetWas%20Ball%20Overlap _bwasballoverlap)
      (if _bwasballoverlap
        (Utilities|FlowControl|Delay 0.5)
        (Variables|Default|SetWas%20Ball%20Overlap false)
        (Utilities|FlowControl|Sequence
          (:then_0
            (Utilities|FlowControl|Delay (Variables|SoccerGame|GetScoreAnnounceDelay))
            (Variables|SoccerGame|SetScore_Team_B (+ (Variables|SoccerGame|GetScore_Team_B) 1))
            ;[1 1 1] Check quest progress
            (if (Variables|SoccerGame|GetbScoredWithToy)
              (Quest|SendComplexCustomStatEvent _returnvalue_2 0 (Variables|Default|GetQuestToySourceTag) "(GameplayTags=)" 1)
              (else
                (Quest|SendComplexCustomStatEvent _returnvalue_2 0 (GameplayTags|MakeGameplayTagContainerfromArray (Variables|Default|GetQuestSourceTags)) "(GameplayTags=)" 1)))
            ;[1 1 1] Set Scoreboard Score
            (Utilities|FlowControl|Delay (Variables|SoccerGame|GetScoreEvaluateDelay))
            (CallFunction|EvaluateScore))
          (:then_1
            (CallFunction|SomeoneScored)
            (GameplayEffects|ApplyGameplayEffectToSelf :self (Ability|GetAbilitySystemComponent (Variables|QuestStuff|GetPlayerWhoScored)) :GameplayEffectClass "/BattlePassS22/Items/QuestAbilities/Quest_ScoreSlideEmote/GE_RecentlyScored.GE_RecentlyScored_C" :Level 0.0)))))))

(event OnComponentBeginOverlap(Team_B_GoalOverlap)
  (params (OverlappedComponent :type Primitive%20Component%20Object%20Reference) (OtherActor :type Actor%20Object%20Reference) (OtherComp :type Primitive%20Component%20Object%20Reference) (OtherBodyIndex :type Integer) (bFromSweep :type Boolean) (SweepResult :type Hit%20Result%20Structure%20%28by%20ref%29))
  (bind _playerwhoscored (Variables|QuestStuff|GetPlayerWhoScored))
  (bind _returnvalue (Fortnite|GetFortPlayerControllerfromActor _playerwhoscored))
  (bind _returnvalue_2 (Quests|GetQuestManager _returnvalue "Invalid"))
  ;[1 1 1] Player Scored on Goal B
  (Networking|SwitchHasAuthority
    (:Authority
      (bind _bwasballoverlap (CallFunction|CheckandHandleBallOverlap OtherActor))
      (Variables|Default|SetWas%20Ball%20Overlap _bwasballoverlap)
      (if _bwasballoverlap
        (Utilities|FlowControl|Delay 0.5)
        (Variables|Default|SetWas%20Ball%20Overlap false)
        (Utilities|FlowControl|Sequence
          (:then_0
            (Utilities|FlowControl|Delay (Variables|SoccerGame|GetScoreAnnounceDelay))
            (Variables|SoccerGame|SetScore_Team_A (+ (Variables|SoccerGame|GetScore_Team_A) 1))
            ;[1 1 1] Check quest progress
            (if (Variables|SoccerGame|GetbScoredWithToy)
              (Quest|SendComplexCustomStatEvent _returnvalue_2 0 (Variables|Default|GetQuestToySourceTag) "(GameplayTags=)" 1)
              (else
                (Quest|SendComplexCustomStatEvent _returnvalue_2 0 (GameplayTags|MakeGameplayTagContainerfromArray (Variables|Default|GetQuestSourceTags)) "(GameplayTags=)" 1)))
            ;[1 1 1] Set Scoreboard Score
            (Utilities|FlowControl|Delay (Variables|SoccerGame|GetScoreEvaluateDelay))
            (CallFunction|EvaluateScore))
          (:then_1
            (CallFunction|SomeoneScored)
            (GameplayEffects|ApplyGameplayEffectToSelf :self (Ability|GetAbilitySystemComponent (Variables|QuestStuff|GetPlayerWhoScored)) :GameplayEffectClass "/BattlePassS22/Items/QuestAbilities/Quest_ScoreSlideEmote/GE_RecentlyScored.GE_RecentlyScored_C" :Level 0.0)))))))

(event Custom|SomeoneScored
  ;[1 1 1] Broadcast Quest Completion when someone scores
  (if (Utilities|Equal(Object) (Data|GetWeaponData (Variables|/Script/FortniteGame.FortPawn|GetCurrentWeapon (Variables|QuestStuff|GetPlayerWhoScored))) (Variables|Default|GetVuvuzela%20WID))
    (CallFunction|PlayVuvuzulaFX)
    (else
      _))
  (Development|PrintString "Scored" false false)
  (if (not (Variables|QuestStuff|GetCreativeMode))
    (Quest|SendComplexCustomStatEvent (Quests|GetQuestManager (Fortnite|GetFortPlayerControllerfromActor (Variables|QuestStuff|GetPlayerWhoScored)) "Invalid") 0 "(GameplayTags=((TagName=\"Athena.Quests.ScoreSoccerGoal\")))" "(GameplayTags=)" 1)))

(event Custom|PlayVuvuzulaFX
  ;[1 1 1] Play Specific Goal FX if you are carrying the Vuvuzula when you score
  (Networking|SwitchHasAuthority
    (:Remote
      (Utilities|FlowControl|Delay 1.0)
      (Effects|SpawnEmitterAttached :EmitterTemplate (Variables|VuvuzulaFXandSound|GetVuvuzulaSpecificEmitter) :AttachToComponent (Variables|/Script/FortniteGame.FortWeapon|GetWeaponMesh (Variables|/Script/FortniteGame.FortPawn|GetCurrentWeapon (Variables|QuestStuff|GetPlayerWhoScored))) :AttachPointName "None" :Scale "1.000000,1.000000,1.000000" :LocationType "SnapToTarget" :bAutoDestroy true :PoolingMethod "None" :bAutoActivate true)
      (Audio|PlaySoundatLocation (Variables|VuvuzulaFXandSound|GetVuvuzulaSpecificSound) (Transformation|GetActorLocation)))))

(event Custom|BallRespawnFX
  (bind _returnvalue (Transformation|GetActorLocation))
  (bind _ball_respawn_location (Variables|SoccerGame|GetBall_Respawn_Location))
  (bind _returnvalue_2 (+ _returnvalue _ball_respawn_location))
  ;[1 1 1] Respawn Ball FX
  (Effects|SpawnEmitteratLocation (Variables|SoccerGameFXandSound|GetPS_RespawnBall) _returnvalue_2)
  (Audio|PlaySoundatLocation (Utilities|Select :Index (Variables|SoccerGame|GetSoccerGameType) :NewEnumerator0 (Variables|SoccerGameFXandSound|GetS_RespawnBall_Small) :NewEnumerator1 (Variables|SoccerGameFXandSound|GetS_RespawnBall) :NewEnumerator2 (Variables|SoccerGameFXandSound|GetS_RespawnBall_Stadium)) _returnvalue_2))

(event Custom|RestartMatch
  ;[1 1 1] Restart Match and Reset Vars
  (Utilities|FlowControl|Delay (Variables|SoccerGame|GetMatchRestartDelay))
  (Variables|SoccerGame|SetScore_Team_A 0)
  (Variables|SoccerGame|SetScore_Team_B 0)
  (Variables|SoccerGame|SetWinningTeam 0)
  (CallFunction|RespawnBall))

(event OnComponentEndOverlap(Ball_DestroyOverlap)
  (params (OverlappedComponent :type Primitive%20Component%20Object%20Reference) (OtherActor :type Actor%20Object%20Reference) (OtherComp :type Primitive%20Component%20Object%20Reference) (OtherBodyIndex :type Integer))
  ;[1 1 1] Check if ball leaves the field, destroy and respawn a new one
  (Utilities|FlowControl|Delay 0.2)
  (if (not (Variables|Default|GetWas%20Ball%20Overlap))
    (Networking|SwitchHasAuthority
      (:Authority
        (bind _asathena_soccerball_redoasis (Utilities|Casting|CastToAthena_Soccerball_Redoasis OtherActor)
          (:then
            (Gameplay|SilentDie _asathena_soccerball_redoasis false :!class /Script/FortniteGame.BuildingActor)
            (CallFunction|RespawnBall))
          (:CastFailed
            (Development|PrintString "NO BALL FOUND")))))))

(fn CheckAndHandleBallOverlap
  (params (Actor :type Actor%20Object%20Reference))
  (bind _asathena_soccerball_redoasis (Utilities|Casting|CastToAthena_Soccerball_Redoasis Actor)
    (:then
      (Variables|QuestStuff|SetPlayerWhoScored (Variables|/BattlePassS22/QuestActors/Redoasis/Athena_Soccerball_Redoasis.Athena_Soccerball_Redoasis_C|GetLastToKick _asathena_soccerball_redoasis))
      (Gameplay|SilentDie _asathena_soccerball_redoasis false :!class /Script/FortniteGame.BuildingActor)
      (Variables|SoccerGame|SetScoreLocation (Transformation|GetActorLocation _asathena_soccerball_redoasis))
      (Variables|SoccerGame|SetbScoredWithToy false))
    (:CastFailed
      (bind _asb_toy_soccer_ball_base (Utilities|Casting|CastToB_Toy_SoccerBall_Base Actor)
        (:then
          (Variables|QuestStuff|SetPlayerWhoScored (Game|Player|GetPlayerPawn (Fortnite|GetFortPlayerControllerfromActor (Actor|GetOwner _asb_toy_soccer_ball_base))))
          (Gameplay|SilentDie _asb_toy_soccer_ball_base false :!class /Script/FortniteGame.BuildingActor)
          (Variables|SoccerGame|SetScoreLocation (Transformation|GetActorLocation _asb_toy_soccer_ball_base))
          (Variables|SoccerGame|SetbScoredWithToy true))
        (:CastFailed
          (return false)))))
  (return true))

(fn EvaluateScore
  (Networking|SwitchHasAuthority
    (:Authority
      (if (== (Variables|SoccerGame|GetScore_Team_A) (Variables|SoccerGame|GetWinScoreLimit))
        (Variables|SoccerGame|SetWinningTeam 1)
        (elif (== (Variables|SoccerGame|GetScore_Team_B) (Variables|SoccerGame|GetWinScoreLimit))
          (Variables|SoccerGame|SetWinningTeam 2)
          (elif (Variables|SoccerGame|GetbScoredWithToy)
            _
            (else
              (CallFunction|RespawnBall))))))))

(fn InitializeGoal
  (params (GoalOverlapComponent :type Box%20Collision%20Object%20Reference) (GoalLocation :type Vector) (ShouldHide :type Boolean))
  (if (not ShouldHide)
    (Transformation|SetRelativeLocation GoalOverlapComponent (Variables|GetGoalLocation) false false)
    (Components|Box|SetBoxExtent GoalOverlapComponent (Variables|SoccerGame|GetGoalExtent) true)))

(fn InitializeGoalCollision
  (params (GoalOverlapComponent :type Box%20Collision%20Object%20Reference) (Hide :type Boolean))
  (Collision|SetCollisionEnabled GoalOverlapComponent (select Hide "NoCollision" "QueryOnly")))

(fn InitializeSign
  (params (ShouldHide :type Boolean))
  (if ShouldHide
    (Rendering|SetVisibility (Variables|Default|GetSM_ScoreSign) false false)
    (else
      (Transformation|SetRelativeTransform (Variables|Default|GetSM_ScoreSign) (Variables|SoccerGame|GetScoreSign_Transform) false false)
      (bind _returnvalue (Rendering|Material|CreateDynamicMaterialInstance (Variables|Default|GetSM_ScoreSign) 1 0 "None" :!class /Script/Engine.PrimitiveComponent))
      (Variables|SoccerGameFXandSound|SetMID_Score_TeamA _returnvalue)
      (bind _returnvalue_2 (Rendering|Material|CreateDynamicMaterialInstance (Variables|Default|GetSM_ScoreSign) 2 0 "None" :!class /Script/Engine.PrimitiveComponent))
      (Variables|SoccerGameFXandSound|SetMID_Score_TeamB _returnvalue_2)
      (Rendering|Material|SetScalarParameterValue (Variables|SoccerGameFXandSound|GetMID_Score_TeamA) "Score" 1 :!class /Script/Engine.MaterialInstanceDynamic)
      (Rendering|Material|SetScalarParameterValue (Variables|SoccerGameFXandSound|GetMID_Score_TeamB) "Score" 2 :!class /Script/Engine.MaterialInstanceDynamic))))

(fn InitializeSignCollision
  (if (Variables|Default|GetHideSign)
    (Collision|SetCollisionEnabled (Variables|Default|GetSM_ScoreSign) "NoCollision")
    (else
      (Collision|SetCollisionEnabled (Variables|Default|GetSM_ScoreSign) "QueryOnly"))))

(fn OnRep_Score_Team_A
  (bind _winscorelimit (Variables|SoccerGame|GetWinScoreLimit))
  (bind _score_team_a (Variables|SoccerGame|GetScore_Team_A))
  (bind _returnvalue (== _score_team_a _winscorelimit))
  (bind _score_team_a_2 (Variables|SoccerGame|GetScore_Team_A))
  (Rendering|Material|SetScalarParameterValue (Variables|SoccerGameFXandSound|GetMID_Score_TeamA) "Score" (Math|Conversions|ToFloat(Integer) _score_team_a_2) :!class /Script/Engine.MaterialInstanceDynamic)
  (if (== _score_team_a_2 0)
    (Rendering|Material|SetScalarParameterValue (Variables|SoccerGameFXandSound|GetMID_Score_TeamA) "Win" 0.0 :!class /Script/Engine.MaterialInstanceDynamic)
    (else
      (Audio|PlaySoundatLocation (Variables|SoccerGameFXandSound|GetS_ScoreUp) (Variables|SoccerGame|GetTeam_A_GoalLocation) "0, 0, 0" 1.0 (select _returnvalue 1.5 1))
      (Rendering|Material|SetScalarParameterValue (Variables|SoccerGameFXandSound|GetMID_Score_TeamA) "Win" (Math|Conversions|ToFloat(Boolean) _returnvalue) :!class /Script/Engine.MaterialInstanceDynamic))))

(fn OnRep_Score_Team_B
  (bind _winscorelimit (Variables|SoccerGame|GetWinScoreLimit))
  (bind _score_team_b (Variables|SoccerGame|GetScore_Team_B))
  (bind _returnvalue (== _score_team_b _winscorelimit))
  (bind _score_team_b_2 (Variables|SoccerGame|GetScore_Team_B))
  (Rendering|Material|SetScalarParameterValue (Variables|SoccerGameFXandSound|GetMID_Score_TeamB) "Score" (Math|Conversions|ToFloat(Integer) _score_team_b_2) :!class /Script/Engine.MaterialInstanceDynamic)
  (if (== _score_team_b_2 0)
    (Rendering|Material|SetScalarParameterValue (Variables|SoccerGameFXandSound|GetMID_Score_TeamB) "Win" 0.0 :!class /Script/Engine.MaterialInstanceDynamic)
    (else
      (Audio|PlaySoundatLocation (Variables|SoccerGameFXandSound|GetS_ScoreUp) (Variables|SoccerGame|GetTeam_B_GoalLocation) "0, 0, 0" 1.0 (select _returnvalue 1.5 1))
      (Rendering|Material|SetScalarParameterValue (Variables|SoccerGameFXandSound|GetMID_Score_TeamB) "Win" (Math|Conversions|ToFloat(Boolean) _returnvalue) :!class /Script/Engine.MaterialInstanceDynamic))))

(fn OnRep_ScoreLocation
  (Effects|SpawnEmitteratLocation (Variables|SoccerGameFXandSound|GetPS_ScoredGoal) (Variables|SoccerGame|GetScoreLocation))
  (Audio|PlaySoundatLocation (Utilities|Select :Index (Variables|SoccerGame|GetSoccerGameType) :NewEnumerator0 (Variables|SoccerGameFXandSound|GetS_ScoredGoal_Small) :NewEnumerator1 (Variables|SoccerGameFXandSound|GetS_ScoredGoal) :NewEnumerator2 (Variables|SoccerGameFXandSound|GetS_ScoredGoal_Stadium)) (Variables|SoccerGame|GetScoreLocation)))

(fn OnRep_WinningTeam
  (bind _returnvalue (Transformation|GetActorLocation))
  (bind _returnvalue_2 (Transformation|GetActorLocation))
  (bind _team_a_goallocation (Variables|SoccerGame|GetTeam_A_GoalLocation))
  (bind _returnvalue_3 (Math|Vector|RotateVector _team_a_goallocation (Transformation|GetActorRotation self)))
  (bind _team_b_goallocation (Variables|SoccerGame|GetTeam_B_GoalLocation))
  (bind _returnvalue_4 (Math|Vector|RotateVector _team_b_goallocation (Transformation|GetActorRotation self)))
  (bind _self self)
  (bind _returnvalue_5 (Transformation|GetActorRotation _self))
  (bind _returnvalue_6 (+ _returnvalue _returnvalue_4))
  (bind _returnvalue_7 (+ _returnvalue_2 _returnvalue_3))
  (bind _soccergametype (Variables|SoccerGame|GetSoccerGameType))
  (bind _s_wongame_stadium (Variables|SoccerGameFXandSound|GetS_WonGame_Stadium))
  (bind _s_wongame (Variables|SoccerGameFXandSound|GetS_WonGame))
  (bind _s_wongame_small (Variables|SoccerGameFXandSound|GetS_WonGame_Small))
  (bind _returnvalue_8 (Utilities|Select :Index _soccergametype :NewEnumerator0 _s_wongame_small :NewEnumerator1 _s_wongame :NewEnumerator2 _s_wongame_stadium))
  (switch int (Variables|SoccerGame|GetWinningTeam)
    (:1
      (Effects|SpawnEmitteratLocation (Variables|SoccerGameFXandSound|GetPS_WonGame) _returnvalue_7)
      (Audio|PlaySoundatLocation _returnvalue_8 _returnvalue_7))
    (:2
      (Effects|SpawnEmitteratLocation (Variables|SoccerGameFXandSound|GetPS_WonGame) _returnvalue_6)
      (Audio|PlaySoundatLocation _returnvalue_8 _returnvalue_6)))
  (Networking|SwitchHasAuthority
    (:Authority
      (CallFunction|RestartMatch))))

(fn RespawnBall
  (if (Variables|Default|GetShouldSpawnBall)
    ;[1 1 1] Check for Soccerballs overlapping spawnpoint and destroy them
    (for _array_element (Collision|GetOverlappingActors (Variables|Default|GetBallOverlapCheck) "/Game/Building/ActorBlueprints/Prop/PhysicsBall_Master.PhysicsBall_Master_C" :!class /Script/Engine.PrimitiveComponent)
      (bind _asathena_soccerball (Utilities|Casting|CastToAthena_Soccerball _array_element))
      (Gameplay|SilentDie _asathena_soccerball false :!class /Script/FortniteGame.BuildingActor)
      (Development|PrintString "Destroyed" false false))
    (Game|SpawnActorfromClass "/BattlePassS22/QuestActors/Redoasis/Athena_Soccerball_Redoasis.Athena_Soccerball_Redoasis_C" (Math|Transform|MakeTransform (+ (Transformation|GetActorLocation) (Variables|SoccerGame|GetBall_Respawn_Location))) "AlwaysSpawn" "MultiplyWithRoot" 0 0 0)
    (CallFunction|BallRespawnFX)))

(fn ConstructionScript
  (CallFunction|InitializeGoal (Variables|Default|GetTeam_A_GoalOverlap) (Variables|SoccerGame|GetTeam_A_GoalLocation) (Variables|Default|GetHideGoalA))
  (CallFunction|InitializeGoal (Variables|Default|GetTeam_B_GoalOverlap) (Variables|SoccerGame|GetTeam_B_GoalLocation) (Variables|Default|GetHideGoalB))
  (CallFunction|InitializeSign (Variables|Default|GetHideSign)))
